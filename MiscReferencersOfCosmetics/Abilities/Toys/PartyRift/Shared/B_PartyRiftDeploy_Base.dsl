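; Base thrown toy projectile that spawns a Party Rift where it lands.
; On an authority-side hit, it checks that the game state allows rifts and the surface is static and flat enough, then silently dies and spawns the rift.
; Entering water kills it; post-throw overlaps play a bounce sound.
; It also carries bouncy-ball helpers for replicated feed messages and random material choice via OnRep functions.
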(event EventBeginPlay
  ;[0 1 0] Event BeginPlay - object starts by being thrown.
  (|CallParentFunction)
  (Variables|Default|SetIs%20Throwing true))

(event Collision|EventHit
  (params (MyComp :type Primitive%20Component%20Object%20Reference) (Other :type Actor%20Object%20Reference) (OtherComp :type Primitive%20Component%20Object%20Reference) (bSelfMoved :type Boolean) (HitLocation :type Vector :default "0, 0, 0") (HitNormal :type Vector :default "0, 0, 0") (NormalImpulse :type Vector :default "0, 0, 0") (Hit :type Hit%20Result%20Structure%20%28by%20ref%29))
  (bind (_bblockinghit _binitialoverlap _time _distance _location _impactpoint _normal _impactnormal _physmat _hitactor _hitcomponent) (Collision|BreakHitResult Hit))
  ;[0 0 1] Event Hit - when this object hits something, check if we're not already touching the ground. If what we hit was a world static object, stop this object, kill it, and spawn the Party Rift
  (|CallParentFunction MyComp Other OtherComp bSelfMoved HitLocation HitNormal NormalImpulse Hit)
  (if (not (Variables|Default|GetHasTouchedGround))
    (Networking|SwitchHasAuthority
      (:Authority
        (if _bblockinghit
          (if (Class|FortGameStateAthena|ShouldAllowPartyRift (Fortnite|Athena|GameState|GetGameStateAthenafromContext self))
            (if (CallFunction|LandedonValidSurface _impactnormal (Collision|GetCollisionObjectType _hitcomponent))
              (Variables|Default|SetHasTouchedGround true)
              (Variables|/Script/Engine.MovementComponent|SetVelocity "0, 0, 0" (Variables|Default|GetProjectileMovement))
              (Gameplay|SilentDie false)
              (Game|SpawnActorfromClass (Variables|Default|GetPartyRiftRef) (Math|Transform|MakeTransform _location) "Undefined" "SelectDefaultAtRuntime" (Fortnite|GetFortPlayerControllerfromActor (Variables|Default|GetOwningPawn)) 0))))))))

(event Custom|Client%20On%20Thrown
  ;[1 0.35 0] Client On Thrown - update flag that we are no longer in the act of throwing this object so we can play sounds based off collisions
  (Utilities|FlowControl|DoOnce false
    (:Completed
      (Variables|Default|SetIs%20Throwing false))))

(event Collision|EventActorBeginOverlap
  (params (OtherActor :type Actor%20Object%20Reference))
  ;[0 0 1] Event ActorBeginOverlap - When we start overlapping something, play a bounce sound
  (|CallParentFunction OtherActor)
  (Networking|SwitchHasAuthority
    (:Authority
      (CallFunction|HitPlayer/VehicleBounceSound))))

(event Custom|Hit%20Player/Vehicle%20Bounce%20Sound
  ;[0 0 1] HitPlayer/VehicleBoundSound - Support Collision with Players and Vehicles by playing a bouncing sound.
  (if (not (Variables|Default|GetIs%20Throwing))
    (Utilities|FlowControl|DoOnce false
      (:Completed
        (Audio|PlaySoundatLocation "/BRCosmetics/Sounds/Other/Toys/BouncyBall/Toys_BouncyBall_Bounce_Cue.Toys_BouncyBall_Bounce_Cue" (Transformation|GetActorLocation))
        (Utilities|FlowControl|Delay (Variables|Default|GetBounceSoundDelay))))))

(event OnEnterWaterEvent(WaterInteractionComponent)
  (params (WaterBody :type Fort%20Water%20Body%20Actor%20Object%20Reference) (WaterInteractionComponent :type Fort%20Water%20Interaction%20Component%20Object%20Reference) (bIsFirstBody :type Boolean))
  ;[1 0.35 0] OnEnterWater - kill the projectile when it hits water
  (Gameplay|SilentDie false))

(fn CreateMessage
  (params (ThrowingPlayer :type Fort%20Player%20State%20Object%20Reference) (Bounces :type Integer))
  (if (Variables|/Script/FortniteGame.FortRuntimeOptions|GetbToyMessagingEnabled (Options|GetRuntimeOptions))
    (Variables|Default|SetDisplayMessage (Utilities|Struct|MakeStructBouncyballMinigameMessage ThrowingPlayer Bounces (Variables|Default|GetMessageSN)))
    (Variables|Default|SetMessageSN (Math|Byte|Byte+Byte (Variables|Default|GetMessageSN) 1))))

(fn LandedOnValidSurface
  (params (SurfaceNormal :type Vector) (CollisionType :type ECollisionChannel%20Enum :default "ECC_WorldStatic"))
  ;[0 0 1] LandedOnValidSurface - returns whether or not the projectile hit a relatively flat surface that would be considered a valid spot to spawn our Party Rift
  (return (and (Utilities|Enum|Equal(Enum) CollisionType "ECC_WorldStatic") (<= (Math|Trig|Acos(Degrees) (Math|Vector|DotProduct SurfaceNormal (Math|Vector|VectorUp))) (LocalVariables|GetMaxAllowedSurfaceSlopeInDegrees)))))

(fn OnRep_Bounces)

(fn OnRep_CheckQuest)

(fn OnRep_DisplayMessage
  (bind _returnvalue (BlueprintContext|GetFortToyContext))
  (bind (_displaymessage_playerwhothrew_22_ef5d780b4020e911ab429d9ba6b977ba _displaymessage_bounces_28_a600d2a048cb01620b914bb4107a6cd5) (Utilities|Struct|BreakStructBouncyballMinigameMessage (Variables|GetDisplayMessage_PlayerWhoThrew_22_EF5D780B4020E911AB429D9BA6B977BA)))
  (Networking|SwitchHasAuthority
    (:Remote
      (Utilities|IsValid _displaymessage_playerwhothrew_22_ef5d780b4020e911ab429d9ba6b977ba
        (:Is%20Valid
          (bind _returnvalue_2 (Class|FortToyContext|FormatPlayerNameforDisplay (BlueprintContext|GetFortToyContext) _displaymessage_playerwhothrew_22_ef5d780b4020e911ab429d9ba6b977ba "ToyOwner"))
          (bind _formatstring (CallFunction|PickFormatString))
          (Utilities|IsValid _returnvalue
            (:Is%20Valid
              (if (not (Variables|Default|GetPrintedMessage))
                (Class|FortToyContext|ConditionalPrinttoMessageFeed _returnvalue _displaymessage_playerwhothrew_22_ef5d780b4020e911ab429d9ba6b977ba (Transformation|GetActorLocation) (select (> (Math|Vector|VectorLength (Transformation|GetActorLocation)) (Variables|Default|GetMinDistanceForTeamRelevance)) "SquadAndTeamMembers" "SquadMembersOnly") (Utilities|Text|FormatText _formatstring _returnvalue_2 _displaymessage_bounces_28_a600d2a048cb01620b914bb4107a6cd5))
                (Variables|Default|SetPrintedMessage true)))))))))

(fn OnRep_MaterialIndex
  (Rendering|Material|SetMaterial (Variables|Default|GetBall) 0 (Utilities|Select :Index (Variables|Default|GetMaterialIndex) :Option%200 (Variables|Materials|GetMaterial1) :Option%201 (Variables|Materials|GetMaterial2) :Option%202 (Variables|Materials|GetMaterial3)))
  (Development|PrintString (Utilities|String|ToString(Integer) (Variables|Default|GetMaterialIndex))))

(fn PickFormatString
  (return (Variables|Default|GetFormatString)))

(fn ConstructionScript
  (|CallParentFunction)
  (Networking|CanExecuteCosmeticEvents
    (:False
      (Variables|Default|SetMaterialIndex (Math|Random|RandomIntegerinRange 0 2)))))
